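# Physics asset for the female medium Grasshopper player cosmetic body, built around its dynamic left and right cloth flaps.
# Two three-bone capsule chains hang from a small Dyn_Master_ClothFlap body on the pelvis, jointed with limited swing and very narrow twist.
# Pelvis, thigh and calf capsules form the lower body around the flaps, while a large flat box sits on the root.
# Joints linking root, pelvis, thighs and master flap have free rotation.
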PhysicsAsset /BRCosmetics/Characters/Player/Female/Medium/Bodies/F_MED_Grasshopper/Meshes/F_MED_Grasshopper_Physics
bodies {
 Dyn_Flap_l1 [{shapeType=Capsule, length=10, radius=5, shapeName=Dyn_Flap_l1_capsule, center=<9.688 7.4 -0.0001>, rotation=<rot 90 0.0395 0.0395>}]
 Dyn_Flap_l2 [{shapeType=Capsule, length=10, radius=5, shapeName=Dyn_Flap_l2_capsule, center=<2.0537 3.8 0>, rotation=<rot 90 0 0>}]
 Dyn_Flap_l3 [{shapeType=Capsule, length=10, radius=5, shapeName=Dyn_Flap_l3_capsule, center=<5 3.8 0>, rotation=<rot 90 0 0>}]
 Dyn_Flap_r1 [{shapeType=Capsule, length=10, radius=5, shapeName=Dyn_Flap_r1_capsule, center=<9.6884 7.4 0>, rotation=<rot 90 0 0>}]
 Dyn_Flap_r2 [{shapeType=Capsule, length=10, radius=5, shapeName=Dyn_Flap_r2_capsule, center=<2.0538 3.8 0>, rotation=<rot 90 0 0>}]
 Dyn_Flap_r3 [{shapeType=Capsule, length=10, radius=5, shapeName=Dyn_Flap_r3_capsule, center=<5 3.8 0>, rotation=<rot 90 0 0>}]
 Dyn_Master_ClothFlap [{shapeType=Capsule, length=0.505, radius=0.505, shapeName=Dyn_Master_ClothFlap_capsule, rotation=<rot 0 -90 0>}]
 calf_l [{shapeType=Capsule, length=26.4294, radius=6.5957, center=<-18.4535 -3.0514 -0.4747>, rotation=<rot 0 88.6683 -89.1174>}]
 calf_r [{shapeType=Capsule, length=26.479, radius=6.6041, center=<19.286 2.927 0.7096>, rotation=<rot 0 88.6429 -89.0288>}]
 pelvis [{shapeType=Capsule, length=16.1818, radius=7.1236, shapeName=pelvis_capsule, center=<-6.6148 -5.5167 0>}]
 root [{shapeType=Box, extentX=70.5, extentY=83.5, extentZ=6.51, shapeName=root_box, center=<0.0001 -18.1379 0>, rotation=<rot 0 -90 0>}]
 thigh_l [{shapeType=Capsule, length=35.83, radius=8.5, center=<-21.012 -1.0146 0.6765>, rotation=<rot 0 -85.6628 -86.9891>}]
 thigh_r [{shapeType=Capsule, length=35.83, radius=8.5, center=<21.0265 1.2083 -0.6117>, rotation=<rot 0 -85.4799 -87.1741>}]
}
constraints [
 {bone1Name=calf_l, bone2Name=thigh_l, swing1LimitDegrees=45, swing2LimitDegrees=45, twistLimitDegrees=45}
 {bone1Name=calf_r, bone2Name=thigh_r, swing1LimitDegrees=45, swing2LimitDegrees=45, twistLimitDegrees=45}
 {bone1Name=Dyn_Flap_r1, bone2Name=Dyn_Master_ClothFlap, swing1LimitDegrees=20, swing2LimitDegrees=50, twistLimitDegrees=1}
 {bone1Name=Dyn_Flap_r2, bone2Name=Dyn_Flap_r1, swing1LimitDegrees=80, swing2LimitDegrees=20, twistLimitDegrees=1}
 {bone1Name=Dyn_Flap_r3, bone2Name=Dyn_Flap_r2, swing1LimitDegrees=80, swing2LimitDegrees=20, twistLimitDegrees=1}
 {bone1Name=Dyn_Flap_l1, bone2Name=Dyn_Master_ClothFlap, swing1LimitDegrees=20, swing2LimitDegrees=50, twistLimitDegrees=1}
 {bone1Name=Dyn_Flap_l2, bone2Name=Dyn_Flap_l1, swing1LimitDegrees=80, swing2LimitDegrees=20, twistLimitDegrees=1}
 {bone1Name=Dyn_Flap_l3, bone2Name=Dyn_Flap_l2, swing1LimitDegrees=80, swing2LimitDegrees=20, twistLimitDegrees=1}
 {
  bone1Name=thigh_l
  bone2Name=pelvis
  swing1LimitDegrees=45
  swing1Motion=Free
  swing2LimitDegrees=45
  swing2Motion=Free
  twistLimitDegrees=45
  twistMotion=Free
 }
 {
  bone1Name=thigh_r
  bone2Name=pelvis
  swing1LimitDegrees=45
  swing1Motion=Free
  swing2LimitDegrees=45
  swing2Motion=Free
  twistLimitDegrees=45
  twistMotion=Free
 }
 {
  bone1Name=Dyn_Master_ClothFlap
  bone2Name=pelvis
  swing1LimitDegrees=45
  swing1Motion=Free
  swing2LimitDegrees=45
  swing2Motion=Free
  twistLimitDegrees=45
  twistMotion=Free
 }
 {
  bone1Name=pelvis
  bone2Name=root
  swing1LimitDegrees=45
  swing1Motion=Free
  swing2LimitDegrees=45
  swing2Motion=Free
  twistLimitDegrees=45
  twistMotion=Free
 }
]
