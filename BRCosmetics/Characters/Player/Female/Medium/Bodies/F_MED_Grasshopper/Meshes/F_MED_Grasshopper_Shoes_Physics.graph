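# Physics asset for the shoes mesh of the F_MED_Grasshopper female medium character cosmetic, covering the left and right shoelace bones.
# Each foot and each Dyn_ShoeLace bone carries a small capsule body.
# Each lace is constrained to its foot with narrow, equal swing limits and almost no twist, allowing slight secondary motion.
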PhysicsAsset /BRCosmetics/Characters/Player/Female/Medium/Bodies/F_MED_Grasshopper/Meshes/F_MED_Grasshopper_Shoes_Physics
bodies {
 Dyn_ShoeLace_l [{shapeType=Capsule, length=2, radius=1, shapeName=Dyn_ShoeLace_l_capsule, center=<0.5 0 0>, rotation=<rot 90 0 0>}]
 Dyn_ShoeLace_r [{shapeType=Capsule, length=2, radius=1, shapeName=Dyn_ShoeLace_r_capsule, center=<0.5 0 0>, rotation=<rot 90 0 0>}]
 foot_l [{shapeType=Capsule, length=0.5, radius=0.5, shapeName=foot_l_capsule}]
 foot_r [{shapeType=Capsule, length=0.5, radius=0.5, shapeName=foot_r_capsule}]
}
constraints [
 {bone1Name=Dyn_ShoeLace_r, bone2Name=foot_r, swing1LimitDegrees=20, swing2LimitDegrees=20, twistLimitDegrees=1}
 {bone1Name=Dyn_ShoeLace_l, bone2Name=foot_l, swing1LimitDegrees=20, swing2LimitDegrees=20, twistLimitDegrees=1}
]
