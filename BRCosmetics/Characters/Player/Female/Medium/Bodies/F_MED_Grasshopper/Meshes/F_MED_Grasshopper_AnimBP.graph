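# Animation Blueprint for the F_MED_Grasshopper female body cosmetic, adding deformation and secondary motion to a pose copied from the attached parent mesh.
# Sculpter and twist Control Rigs feed shoe and body RigidBody simulations, plus a skydiving-only simulation and cloth-flap rotation.
# Weighted rotation constraints tie side flaps to pelvis and leg bones.
# EventGraph backbling checks set cape and backpack flags, nudging the right flap for front-end capes.
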AnimBlueprint /BRCosmetics/Characters/Player/Female/Medium/Bodies/F_MED_Grasshopper/Meshes/F_MED_Grasshopper_AnimBP
target_skeleton=&/F_MED_Grasshopper_Skeleton
graphs [
 AnimGraph {nodes [
   n1 = "Transform (Modify) Bone - Bone: Dyn_Master_Flap_r"!(ComponentPose, bAlphaBoolEnabled:bool=True, Translation="0,-1.8,0", Rotation="-5,-15,10", Scale="1,1,1") -> (Pose) {
    alphaInputType=Bool
    rotationMode=BMM_Additive
    rotationSpace=BCS_BoneSpace
    translationMode=BMM_Additive
    translationSpace=BCS_BoneSpace
    alphaBoolBlend {blendInTime=0.2, blendOption=Linear, blendOutTime=0.2}
    boneToModify {boneName=Dyn_Master_Flap_r}
    rotation=<rot -5 -15 10>
    translation {y=-1.8}
   }
   "Transform (Modify) Bone - Bone: Dyn_Master_Flap_l"!(ComponentPose=@n1, bAlphaBoolEnabled:bool=True, Translation="0,-1.8,0", Rotation="-5,-15,0", Scale="1,1,1") -> (Pose) {
    alphaInputType=Bool
    rotationMode=BMM_Additive
    rotationSpace=BCS_BoneSpace
    translationMode=BMM_Additive
    translationSpace=BCS_BoneSpace
    alphaBoolBlend {blendInTime=0.2, blendOption=Linear, blendOutTime=0.2}
    boneToModify {boneName=Dyn_Master_Flap_l}
    rotation {pitch=-5, yaw=-15}
    translation {y=-1.8}
   }
   "Transform (Modify) Bone - Bone: Dyn_Master_Flap_l"(bAlphaBoolEnabled=True, Translation="0,0,0", Rotation="0,0,0", Scale="1,1,1") {
    alphaInputType=Bool
    rotationMode=BMM_Additive
    rotationSpace=BCS_BoneSpace
    translationMode=BMM_Additive
    translationSpace=BCS_BoneSpace
    alphaBoolBlend {blendInTime=0.2, blendOption=Linear, blendOutTime=0.2}
    boneToModify {boneName=Dyn_Master_Flap_l}
   }
   n2 = Control Rig(Source=Copy Pose From Mesh() {bUseAttachedParent=true}) {
    alphaBoolBlend {blendOption=Linear}
    controlRigAssetReference {blueprintRigClass=&/F_MED_Grasshopper_Deform_CtrlRig_sculpters.F_MED_Grasshopper_Deform_CtrlRig_sculpters_C}
   }
   n3 = Control Rig(Source=@n2) {
    alphaBoolBlend {blendOption=Linear}
    controlRigAssetReference {blueprintRigClass=&/F_MED_Grasshopper_Deform_CtrlRig_twists.F_MED_Grasshopper_Deform_CtrlRig_twists_C}
   }
   n4 = RigidBody(ComponentPose=Local To Component(LocalPose=@n3)) {
    bClampLinearTranslationLimitToRefPose=true
    bFreezeIncomingPoseOnStart=true
    bOverrideWorldGravity=true
    bUseExternalClothCollision=true
    overridePhysicsAsset=&/F_MED_Grasshopper_Shoes_Physics
    simulationSpace=BaseBoneSpace
    alphaBoolBlend {blendOption=Linear}
    baseBoneRef {boneName=root}
    componentAppliedLinearAccClamp=<3000 3000 3000>
    componentLinearAccScale=<0.1 0.1 0.1>
    componentLinearVelScale=<0.1 0.1 0.1>
    overrideWorldGravity {z=-400}
   }
   n5 = RigidBody(ComponentPose=@n4) {
    bClampLinearTranslationLimitToRefPose=true
    bOverrideWorldGravity=true
    bUseExternalClothCollision=true
    overridePhysicsAsset=&/F_MED_Grasshopper_Physics
    simulationSpace=BaseBoneSpace
    alphaBoolBlend {blendOption=Linear}
    baseBoneRef {boneName=pelvis}
    componentAppliedLinearAccClamp=<5000 5000 5000>
    componentLinearAccScale=<1 0.35 0.5>
    componentLinearVelScale=<1 0.35 0.5>
    overrideWorldGravity {z=-980}
   }
   n6 = Get bIsSkyDiving()
   n7 = RigidBody!(ComponentPose=@n5, bAlphaBoolEnabled:bool=@n6, ExternalForce) -> (Pose) {
    alphaInputType=Bool
    bClampLinearTranslationLimitToRefPose=true
    bOverrideWorldGravity=true
    overridePhysicsAsset=&F_MED_Grasshopper_Physics
    simulationSpace=BaseBoneSpace
    alphaBoolBlend {blendInTime=1, blendOption=Linear, blendOutTime=1}
    baseBoneRef {boneName=pelvis}
    componentAppliedLinearAccClamp=<3000 3000 3000>
    componentLinearAccScale=<1 0.1 0.1>
    componentLinearVelScale=<1 0.1 0.1>
    overrideWorldGravity {z=-400}
   }
   n8 = "Transform (Modify) Bone - Bone: Dyn_Master_ClothFlap"!(ComponentPose=@n7, bAlphaBoolEnabled:bool=@n6, Translation="0,0,0", Rotation="0,30,0", Scale="1,1,1") -> (Pose) {
    alphaInputType=Bool
    rotationMode=BMM_Additive
    rotationSpace=BCS_BoneSpace
    alphaBoolBlend {blendInTime=1, blendOption=Linear, blendOutTime=1}
    boneToModify {boneName=Dyn_Master_ClothFlap}
    rotation {yaw=30}
   }
   n9 = "Constraint - Dyn_Master_Flap_r (RRR)"!(ComponentPose=@n8, bAlphaBoolEnabled:bool=True, ConstraintWeights_0=0.35, ConstraintWeights_1:float=0.3, ConstraintWeights_2:float=0.35) -> (Pose) {
    alphaInputType=Bool
    alphaBoolBlend {blendInTime=0.2, blendOption=Linear, blendOutTime=0.2}
    boneToModify {boneName=Dyn_Master_Flap_r}
    constraintSetup [
     {transformType=Rotation, targetBone {boneName=thigh_r}}
     {transformType=Rotation, targetBone {boneName=pelvis}}
     {transformType=Rotation, targetBone {boneName=calf_r}}
    ]
    constraintWeights [1,1,1]
   }
   n10 = "Constraint - Dyn_Master_Flap_l (RRR)"!(ComponentPose=@n9, bAlphaBoolEnabled:bool=True, ConstraintWeights_0=0.35, ConstraintWeights_1:float=0.3, ConstraintWeights_2:float=0.35) -> (Pose) {
    alphaInputType=Bool
    alphaBoolBlend {blendInTime=0.2, blendOption=Linear, blendOutTime=0.2}
    boneToModify {boneName=Dyn_Master_Flap_l}
    constraintSetup [
     {transformType=Rotation, targetBone {boneName=thigh_l}}
     {transformType=Rotation, targetBone {boneName=pelvis}}
     {transformType=Rotation, targetBone {boneName=calf_l}}
    ]
    constraintWeights [1,1,1]
   }
   n11 = "Transform (Modify) Bone - Bone: Dyn_Master_Flap_r"(ComponentPose=@n10, bAlphaBoolEnabled=Get WearingCapeFE() -> (WearingCapeFE:bool), Translation="0,0,0", Rotation="0,-7.5,0", Scale="1,1,1") {
    alphaInputType=Bool
    rotationMode=BMM_Additive
    rotationSpace=BCS_BoneSpace
    translationMode=BMM_Additive
    translationSpace=BCS_BoneSpace
    alphaBoolBlend {blendInTime=0.2, blendOption=Linear, blendOutTime=0.2}
    boneToModify {boneName=Dyn_Master_Flap_r}
    rotation {yaw=-7.5}
   }
   Output Pose(Result=Component To Local(ComponentPose=@n11)) {layerGroup=None}
  ]}
 EventGraph {nodes [
   Event BlueprintUpdateAnimation()
   > n1 = Sequence()
   n2 = CheckForBackbling()
   > Set WearingAtBack(WearingAtBack:bool=OR Boolean!(A=@n2, B=@n2.bIsBackblingACape, C:bool=@n2.bIsBackblingLong) -> (ReturnValue)) -> (Output_Get:bool)
   > Set isWearingBackpack(isWearingBackpack:bool=AND Boolean!(A=@n2, B=NOT Boolean(A=@n2.bIsBackblingACape), C:bool=NOT Boolean(A=@n2.bIsBackblingLong)) -> (ReturnValue)) -> (Output_Get:bool)
   > Set Is Wearing Cape(Is Wearing Cape=OR Boolean!(A=@n2.bIsBackblingACape, B=@n2.bIsBackblingLong) -> (ReturnValue))
   Set WearingCapeFE(execute=@n1.then_1, WearingCapeFE:bool=AND Boolean(A=Get bIsInFrontEnd(), B=Get Is Wearing Cape() -> (Is Wearing Cape:bool), C=true)) -> (Output_Get:bool)
   NOT Boolean(A=false)
   AND Boolean!(A=NOT Boolean(A=Get bRigidBodyEnabled()), B=false) -> (ReturnValue)
   Get bClothEnabled()
  ]}
]
variables [
 Is Wearing Cape {type=boolean}
 isWearingBackpack {type=boolean}
 WearingAtBack {type=boolean}
 WearingCapeFE {type=boolean}
]
